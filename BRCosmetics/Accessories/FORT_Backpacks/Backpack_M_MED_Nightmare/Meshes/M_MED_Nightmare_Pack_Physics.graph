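# Physics asset for the Nightmare medium male backpack cosmetic, giving its jaw and fur pieces simulated secondary motion.
# Box bodies cover a jaw bone, upper and lower fur bones on each side, and the pack bone, to which all others are directly constrained.
# The fur constraints allow moderate swing with slight twist, while the jaw locks twist and one swing axis, leaving a narrow single-axis hinge.
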PhysicsAsset /BRCosmetics/Accessories/FORT_Backpacks/Backpack_M_MED_Nightmare/Meshes/M_MED_Nightmare_Pack_Physics
bodies {
 dyn_jaw [{shapeType=Box, extentX=11.8815, extentY=17.1999, extentZ=12.0968, center=<10.1995 0.0003 -9.4387>}]
 furLower_l [{shapeType=Box, extentX=6, extentY=6, extentZ=1, center=<2.5 0 0>}]
 furLower_r [{shapeType=Box, extentX=6, extentY=6, extentZ=1, center=<2.5 0 0>}]
 fur_l [{shapeType=Box, extentX=8, extentY=8, extentZ=1, center=<2.5 0 0>}]
 fur_r [{shapeType=Box, extentX=8, extentY=8, extentZ=1, center=<2.5 0 0>}]
 pack [{shapeType=Box, extentX=1, extentY=1, extentZ=1}]
}
constraints [
 {bone1Name=fur_l, bone2Name=pack, swing1LimitDegrees=35, swing2LimitDegrees=35, twistLimitDegrees=5}
 {bone1Name=fur_r, bone2Name=pack, swing1LimitDegrees=35, swing2LimitDegrees=35, twistLimitDegrees=5}
 {bone1Name=furLower_l, bone2Name=pack, swing1LimitDegrees=35, swing2LimitDegrees=35, twistLimitDegrees=5}
 {bone1Name=furLower_r, bone2Name=pack, swing1LimitDegrees=35, swing2LimitDegrees=35, twistLimitDegrees=5}
 {bone1Name=dyn_jaw, bone2Name=pack, swing1LimitDegrees=5, swing1Motion=Locked, swing2LimitDegrees=15, twistLimitDegrees=5, twistMotion=Locked}
]
